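# Sound cue for the Karate pickaxe hitting an enemy.
# After a short fixed delay, it mixes a random Karate impact wave with a quieter, pitched-down random player flesh impact.
# It is primed on load, uses the pickaxe impact attenuation and property sheet, and sets owner-limited concurrency that stops the oldest sound.
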SoundCue /BRCosmetics/Sounds/PickAxes/Karate/PickAxe_Karate_EnemyImpact_Cue
root=Delay_1
nodes [
 {id=Delay_1, children [Mixer_1], properties {delayMax=0.1, delayMin=0.1}}
 {id=Mixer_1, children [Modulator_0,Random_0], properties {inputVolume [0.37,0.9]}}
 {id=Modulator_0, children [Random_1], properties {pitchMax=0.85, pitchMin=0.85}}
 {id=Random_0, children [WavePlayer_0,WavePlayer_1,WavePlayer_2,WavePlayer_3,WavePlayer_8], properties {weights [1,1,1,1,1]}}
 {id=Random_1, children [WavePlayer_4,WavePlayer_5,WavePlayer_6,WavePlayer_7], properties {weights [1,1,1,1]}}
 {id=WavePlayer_0, properties {soundWaveAssetPtr=&/PickAxe_Karate_EnemyImpact_01}}
 {id=WavePlayer_1, properties {soundWaveAssetPtr=&/PickAxe_Karate_EnemyImpact_02}}
 {id=WavePlayer_2, properties {soundWaveAssetPtr=&/PickAxe_Karate_EnemyImpact_03}}
 {id=WavePlayer_3, properties {soundWaveAssetPtr=&/PickAxe_Karate_EnemyImpact_04}}
 {id=WavePlayer_4, properties {soundWaveAssetPtr=/Game/Sounds/Fort_Impact_Sounds/Flesh/new_player_impact_nonspat_A_01n}}
 {id=WavePlayer_5, properties {soundWaveAssetPtr=/Game/Sounds/Fort_Impact_Sounds/Flesh/new_player_impact_nonspat_A_02n}}
 {id=WavePlayer_6, properties {soundWaveAssetPtr=/Game/Sounds/Fort_Impact_Sounds/Flesh/new_player_impact_nonspat_A_03n}}
 {id=WavePlayer_7, properties {soundWaveAssetPtr=/Game/Sounds/Fort_Impact_Sounds/Flesh/new_player_impact_nonspat_A_04n}}
 {id=WavePlayer_8, properties {soundWaveAssetPtr=&/PickAxe_Karate_EnemyImpact_05}}
]
properties {
 attenuationSettings=/Game/Sounds/Attenuation/Fort_Pickaxe_Impact_Att
 audioPropertiesSheet=/BRCosmetics/Sounds/Settings/PropertySheets/Pickaxe/APS_Cosmetic_Pickaxe_Impact
 bOverrideConcurrency=true
 bPrimeOnLoad=true
 maxDistance=4300
 soundClassObject=&Fort_Impacts
 volumeMultiplier=0.525
 attenuationOverrides {
  bAttenuateWithLPF=true
  distanceAlgorithm=Logarithmic
  falloffDistance=800
  lPFRadiusMax=800
  lPFRadiusMin=400
  attenuationShapeExtents {x=300}
 }
 concurrencyOverrides {bLimitToOwner=true, maxCount=4, resolutionRule=StopOldest}
 concurrencySet [/Game/Sounds/Concurrency_Rules/OneMaximum_Owner_StopOldest]
}
